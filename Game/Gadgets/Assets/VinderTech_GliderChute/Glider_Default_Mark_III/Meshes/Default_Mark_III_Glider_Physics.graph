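# Physics asset for the Default Mark III VinderTech glider chute mesh, giving the glider a simplified physics shape.
# A single Body_main body holds one large capsule, offset from the bone origin and rotated onto its side.
# It has no constraints, so there is no articulated ragdoll or jointed motion between parts.
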PhysicsAsset /Game/Gadgets/Assets/VinderTech_GliderChute/Glider_Default_Mark_III/Meshes/Default_Mark_III_Glider_Physics
bodies {Body_main [{shapeType=Capsule, length=270.6947, radius=85.8752, center=<0 14.4708 61.4625>, rotation=<rot 0 90.0041 -89.9981>}]}
constraints []
